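; Exotic ColdFern explosion actor for the WildEstate weapons, named as a BGA gameplay actor.
; On BeginPlay it plays the sound cue held in its SoundCue variable at its own location.
; The construction script is empty, and the graph shows no damage or visual logic, only the audio.
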(event EventBeginPlay
  (Audio|PlaySoundatLocation (Variables|Default|GetSoundCue) (Transformation|GetActorLocation)))

(fn ConstructionScript)
