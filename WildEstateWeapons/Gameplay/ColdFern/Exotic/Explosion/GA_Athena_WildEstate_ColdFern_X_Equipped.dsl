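; Equipped gameplay ability for the exotic ColdFern weapon in WildEstate that adds an explosion to one shot per magazine.
; On event activation it stores the instigator and hit location from target data, then updates a bullet count from magazine ammo, or from a wrapping counter under infinite ammo.
; When the count is 1, it applies an area gameplay effect around the hit, spawns the explosion building actor and plays a sound.
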(event Ability|EventActivateAbilityFromEvent
  (params (EventData :type Gameplay%20Event%20Data%20Structure%20%28by%20ref%29))
  (bind (_eventtag _instigator) (Utilities|Struct|BreakGameplayEventData EventData))
  (bind _hitlocation (Variables|Default|GetHitLocation))
  (bind (_valid _location) (Default|GetHitInfo EventData))
  (Variables|Default|SetHitLocation _location)
  (Variables|Default|SetOwningPlayerPawn _instigator)
  (bind _bulletcount (Default|CheckandHandleBulletCount))
  (if (== _bulletcount 1)
    (Fortnite|ApplyGameplayEffectToActorsInRange (Utilities|Array|MakeArray (Variables|Default|GetAoEGE)) (Variables|Default|GetObject%20Types) (Variables|Default|GetHitLocation) 5e+02 true 0 (Ability|GetAbilitySystemComponent (Variables|Default|GetOwningPlayerPawn)))
    (Building|SpawnBuildingGameplayActor "/WildEstateWeapons/Gameplay/ColdFern/Exotic/Explosion/BGA_Athena_WildEstate_ColdFern_X_Explosion.BGA_Athena_WildEstate_ColdFern_X_Explosion_C" (Math|Transform|MakeTransform _hitlocation) (Variables|Default|GetOwningPlayerPawn) true false)
    (Audio|SpawnSoundatLocation (Variables|Default|GetCueExplosionSound) _hitlocation)
    (Ability|EndAbility)
    (else
      (Ability|EndAbility))))

(fn CheckAndHandleBulletCount
  (bind _returnvalue (Ability|GetCurrentSourceWeapon))
  (bind _returnvalue_2 (FortWeapon|Ammo|HasInfiniteMagazineAmmo _returnvalue))
  (bind _returnvalue_3 (not _returnvalue_2))
  (bind _returnvalue_4 (FortWeapon|GetBulletsPerClip _returnvalue))
  ; If not using the infinite ammo config
  (if _returnvalue_3
    (Variables|Default|SetBulletCount (- (FortWeapon|GetBulletsPerClip (Ability|GetCurrentSourceWeapon)) (FortWeapon|GetLocalMagazineAmmoCount (Ability|GetCurrentSourceWeapon))))
    (elif (and _returnvalue_3 (== (+ (FortWeapon|GetLocalMagazineAmmoCount _returnvalue) 1) _returnvalue_4))
      (Variables|Default|SetBulletCount 0)
      ; Check if reached bullets per clip
      (if (>= (Variables|Default|GetBulletCount) _returnvalue_4) :!label _label
        (Variables|Default|SetBulletCount 1)
        (else
          (bind _result (Math|Integer|IncrementInt (Variables|Default|GetBulletCount))
            (:%20%20
              (Variables|Default|SetBulletCount _result)))))
      (else
        (goto _label))))
  (return (Variables|Default|GetBulletCount)))

(fn GetHitInfo
  (params (EventData :type Gameplay%20Event%20Data%20Structure))
  (bind (_eventtag _instigator _target _optionalobject _optionalobject2 _contexthandle _instigatortags _targettags _eventmagnitude _targetdata) (Utilities|Struct|BreakGameplayEventData EventData))
  (for _index (range (Ability|TargetData|GetDataCountfromTargetData _targetdata))
    (if (Collision|BreakHitResult (Ability|TargetData|GetHitResultfromTargetData _targetdata _index))
      (return true (.Location (Collision|BreakHitResult (Ability|TargetData|GetHitResultfromTargetData _targetdata _index))) (.HitActor (Collision|BreakHitResult (Ability|TargetData|GetHitResultfromTargetData _targetdata _index))))))
  (return false "0, 0, 0" 0))
